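# Niagara effects for the GelatoCourage glider: a pulsing jet exhaust, red backlights and ribbon trails attached to its skeletal mesh.
# System timeline curves and ForwardDot drive the flame flares with light, GPU embers and smoke, with front-end overrides.
# Smoke bends against owner velocity, while a partitioned burst builds the backlight sprite grid.
# Soft and tip AnimTrail ribbons follow bone pairs, tinted by VariantPicker and time of day.
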NiagaraSystem /BRCosmetics/Gadgets/Assets/VinderTech_GliderChute/Glider_GelatoCourage/FX/NS_Glider_GelatoCourage
dependencies {
 usedDataInterfaces [
  &NiagaraDataInterfaceCurve
  &NiagaraDataInterfaceColorCurve
  &NiagaraDataInterfaceArrayFloat3
  &NiagaraDataInterfaceVector2DCurve
  &NiagaraDataInterfaceArrayColor
  &NiagaraDataInterfaceCamera
 ]
 usedDynamicInputs [&Divide_Float, &Power]
 usedModules [
  &self:SystemSpawnScript.SystemScriptSource.SystemScriptGraph.NiagaraNodeAssignment_2.SetVariables_#1
  &SystemState
  &Timeline
  &self:SystemSpawnScript.SystemScriptSource.SystemScriptGraph.NiagaraNodeAssignment_1.SetVariables_#2
  &self:SystemSpawnScript.SystemScriptSource.SystemScriptGraph.NiagaraNodeAssignment_0.SetVariables_#3
 ]
}
emitters {
 EmbersGPUHigh {
  simTarget=GPUComputeSim
  emitterUpdateScript [
   &EmitterState()
   &SpawnRate(
    Spawn Probability=&FloatFromCurve(
     FloatCurve=DataInterface() {propertyValues {Curve {keys [
         <at 0 0 RCIM_Constant>
         <at 0.11 1 RCIM_Constant>
         <at 0.25 0 RCIM_Constant>
         <at 0.4 1 RCIM_Constant>
         {arriveTangent=-2.5249, interpMode=RCIM_Constant, leaveTangent=-2.5249, time=0.5, value=0}
        ]}}}
     CurveIndex=$System.Timeline.Time
     Scale Curve=&FN_FrontEndFloatModifier(Gameplay=$User.ForwardDot, Front End=1)
    )
    SpawnRate=50
    Use Spawn Probability=true
   )
  ]
  particleSpawnScript [
   /SetVariables(Particles.PerParticleRandom:NiagaraFloat=&RandomRangeFloat(Minimum=0, Maximum=1)) {isSetParameters=true}
   &InitializeParticle(
    Color Mode=::NewEnumerator0
    Lifetime=&Lerp_Float(A=0.999, B=0.666, Alpha=$Particles.PerParticleRandom)
    Position Offset=$System.JetPosition
    Sprite Size=&Multiply_Vector2D(
     A=<0.5 1>
     B=&MakeVector2D(
      X=&FloatFromCurve(
       FloatCurve=DataInterface() {propertyValues {Curve {keys [<at 0 1>,<at 0.8 2 RCIM_Constant>,<at 0.95 3 RCIM_Constant>]}}}
       CurveIndex=$Particles.PerParticleRandom
      )
      Y=&FloatFromCurve(
       FloatCurve=DataInterface() {propertyValues {Curve {keys [<at 0 1>,<at 0.8 2 RCIM_Constant>,<at 0.95 3>]}}}
       CurveIndex=$Particles.PerParticleRandom
      )
     )
    )
    Sprite Size Mode=::NewEnumerator4
    UsePositionOffset=true
    Write Material Random=true
   )
   &SkeletalMeshLocation(Bone Sampling Mode=::NewEnumerator4, MaskSpawnGroup=false, Sampled Position Offset=<16 0 0>, Skeletal Mesh=$System.Skeletal Mesh)
   &AddVelocity(
    Cone Angle=33
    Cone Axis=<0 1 0>
    Module.UseConeFalloff=false
    Module.UseVelDistribution=false
    Override Seed=false
    Velocity Mode=::NewEnumerator2
    Velocity Speed=&Lerp_Float(A=333, B=777, Alpha=$Particles.PerParticleRandom)
   )
   &ShapeLocation(MaskSpawnGroup=false, Override Seed=false, Sphere Radius=5, UseOwnerScale=true)
  ]
  particleUpdateScript [
   &ParticleState()
   &AccelerationForce()
   &Drag(
    Drag=&FloatFromCurve(
     FloatCurve=DataInterface() {propertyValues {Curve {keys [<at 0 0.5>,<at 0.5 0.5>,<at 1 1>]}}}
     Scale Curve=&RandomRangeFloat(Minimum=7, Maximum=22)
    )
    Use Linear Drag=true
    Use Rotational Drag=false
   )
   /Niagara/Modules/Update/Forces/CurlNoiseForce(
    Noise Frequency=0.04
    Noise Strength=&FloatFromCurve(
     FloatCurve=DataInterface() {propertyValues {Curve {keys [{tangentMode=RCTM_User, time=0, value=0}, {tangentMode=RCTM_User, time=1, value=1}]}}}
     Scale Curve=&Lerp_Float(A=2222, B=444, Alpha=$Particles.PerParticleRandom)
    )
    Pan Noise=true
    Pan Noise Field=<0 0 1>
   )
   &SolveForcesAndVelocity(Clamp Velocity=false, Limit Acceleration=false)
   &ScaleColor(
    Scale RGB=&MakeVectorFromLinearColorRGB(Color=&ColorFromCurve(ColorCurve=DataInterface() {propertyValues {
        AlphaCurve {keys [<at 0 1 RCIM_Cubic>]}
        BlueCurve {keys [<at 0 10>, {arriveTangent=-10, interpMode=RCIM_Cubic, leaveTangent=-10, time=0.1905, value=0}, <at 1 0 RCIM_Cubic>]}
        GreenCurve {keys [<at 0 1.9448>, {arriveTangent=-1.9448, interpMode=RCIM_Cubic, leaveTangent=-1.9448, time=0.1905, value=1.7351}, <at 1 0 RCIM_Cubic>]}
        RedCurve {keys [<at 0 0>, {arriveTangent=1, interpMode=RCIM_Cubic, leaveTangent=1, time=0.1905, value=5}, <at 1 1 RCIM_Cubic>]}
       }}))
    ScaleA=true
    ScaleRGB=true
   )
   &ScaleSpriteSizeBySpeed(
    Max Scale Factor=<1.666 13>
    Min Scale Factor=<1.333 1.333>
    Sample Scale Factor By Curve=true
    Scale Factor Curve=DataInterface() {propertyValues {Curve {keys [<at 0 0>, {arriveTangent=1.4989, interpMode=RCIM_Cubic, tangentMode=RCTM_Break, time=1, value=1}]}}}
    Velocity Threshold=999
   )
   &ScaleSpriteSize(
    Uniform Curve Scale=&MakeFloatFromVector(Vector=$Engine.Owner.Scale, Channel=::NewEnumerator0)
    Uniform Curve Sprite Scale=DataInterface() {propertyValues {Curve {keys [
        {tangentMode=RCTM_User, time=0, value=1}
        {arriveTangent=-0.0405, interpMode=RCIM_Cubic, leaveTangent=-0.0405, tangentMode=RCTM_User, time=0.7009, value=1}
        {arriveTangent=-8.0089, interpMode=RCIM_Cubic, leaveTangent=-8.0089, tangentMode=RCTM_User, time=1, value=0}
       ]}}}
   )
  ]
  rendererClasses [&NiagaraSpriteRendererProperties]
  renderers [{rendererClass=&NiagaraSpriteRendererProperties}]
 }
 NE_Backlights {
  emitterUpdateScript [
   &EmitterState(
    Inactive Response=::NewEnumerator1
    Life Cycle Mode=::NewEnumerator1
    Loop Behavior=::NewEnumerator1
    Loop Duration Mode=::NewEnumerator1
    UseLoopDelay=false
   )
   /SetVariables(Emitter.BackLightStrength:NiagaraFloat=&OneMinusFloat(Float=$User.ForwardDot)) {enabled=false, isSetParameters=true}
   /SetVariables(Emitter.BackLightStrength=1) {isSetParameters=true}
   &SpawnBurst_Instantaneous(Spawn Count=9, Use Spawn Probability=false, UseLoopCountLimit=false)
   &SpawnBurst_Instantaneous(Spawn Group=1, Use Spawn Probability=false, UseLoopCountLimit=false) {module=SpawnBurst_Instantaneous001}
  ]
  particleSpawnScript [
   /SetVariables(Particles.VisibilityTag=&SpawnGroup()) {isSetParameters=true}
   &PartitionParticles(Number Of Partitions=9)
   &InitializeParticle(
    Color=<rgba 2 0 0 0.2>
    Mesh Scale Mode=::NewEnumerator3
    Position Offset=&Add_Vector(
     A=<0 69.5 -8.5>
     B=&SelectVectorFromArray(
      Direct Array Index=$Particles.PartitionParticles.Partition
      Vector Selection Array=DataInterface() {propertyValues {FloatData [<-2 0 0>, <0 0 0>, <2 0 0>, <-2 -0.15 -2>, <0 -0.15 -2>, <2 -0.15 -2>, <-2 -0.3 -4>, <0 -0.3 -4>, <2 -0.3 -4>]}}
     )
    )
    Sprite Size=<15 15>
    Sprite Size Mode=::NewEnumerator4
    UsePositionOffset=true
    Write Material Random=true
   )
   &SpriteFacingAndAlignment(Alignment Coordinate Space=::Local, Sprite Alignment=<0 -1 0>, Sprite Facing=<0 1 0>, WriteAlignment=true, WriteFacing=false)
   &CameraOffset()
   &DynamicMaterialParameters(Module.Param0WriteEnabled=false, Module.Param1WriteEnabled=false, Module.Param2WriteEnabled=false, Module.Param3WriteEnabled=false)
   /SetVariables(Particles.BeamColor:LinearColor=&MultiplyLinearColor(Linear Color=$Particles.Color, Scale Factor=&MakeVector4_FromFloat(Float=0.5))) {isSetParameters=true}
   /SetVariables(
    Particles.FlareAlignment:Vector3f=<0 0 1>
    Particles.FlareCameraOffset=77
    Particles.FlareColor=&MultiplyLinearColor(Linear Color=$Particles.Color, Scale Factor=&MakeVector4_FromFloat(Float=1))
    Particles.FlareFacing:Vector3f=<0 0 1>
    Particles.FlarePosition:NiagaraPosition=$Particles.Position
    Particles.FlareSize:Vector2f=<200 200>
   ) {isSetParameters=true}
   /SetVariables(
    Particles.TeardropGlowColor:LinearColor=&MultiplyLinearColor(Linear Color=$Particles.Color, Scale Factor=&MakeVector4_FromFloat(Float=0.3))
    Particles.TeardropGlowPosition:NiagaraPosition=&AddVectorToPosition(Position=$Particles.Position, Vector=<0 0 -1>)
    Particles.TeardropGlowProperties:Vector4f=&MakeFloat4FromVectorAndFloat(
     "Vector (RGB)"=&Multiply_Vector(
      A=&NormalizeVector(
       Vector To Normalize=/Niagara/DynamicInputs/Vectors/CrossProduct(Vector A:Vector3f=$Particles.FlareAlignment, Vector B:Vector3f=$Particles.FlareFacing, Normalize Incoming Vectors:NiagaraBool=true)
       Safe Normalize=true
       Safe Fallback Threshold=0
       Safe Fallback Vector=<0 0 0.001>
      )
      B=<1 1 1>
     )
     "Float (Alpha)"=25
    )
    Particles.TeardropGlowSize:Vector2f=&Vector2DFromFloat(Value=5)
   ) {isSetParameters=true}
   /SetVariables(
    Particles.GlowColor=&MultiplyLinearColor(Linear Color=$Particles.Color, Scale Factor=&MakeVector4_FromFloat(Float=5))
    Particles.GlowPosition:NiagaraPosition=&AddVectorToPosition(Position=$Particles.Position, Vector=<0 0 -1>)
    Particles.GlowSize=&Vector2DFromFloat(Value=3)
   ) {isSetParameters=true}
  ]
  particleUpdateScript [
   &ParticleState(Kill Particles When Lifetime Has Elapsed=false)
   &ScaleSpriteSize(
    Non-Uniform Curve Index:NiagaraFloat=$User.ForwardDot
    Non-Uniform Curve Sprite Scale=DataInterface() {propertyValues {
      XCurve {keys [{tangentMode=RCTM_User, time=0, value=1}, {tangentMode=RCTM_User, time=1, value=1}]}
      YCurve {keys [{tangentMode=RCTM_User, time=0, value=0.1}, {tangentMode=RCTM_User, time=1, value=1}]}
     }}
    Scale Sprite Size Mode=::NewEnumerator3
   )
   &ScaleColor(Scale Alpha=$Emitter.BackLightStrength, ScaleA=true, ScaleRGB=true)
   &SkeletalMeshLocation(
    Bone Sampling Mode=::NewEnumerator4
    MaskSpawnGroup=false
    Sampled Position Offset=&Add_Vector(
     A=<7.5 0 -10.7>
     B=&SelectVectorFromArray(
      Direct Array Index=$Particles.PartitionParticles.Partition
      Vector Selection Array=DataInterface() {propertyValues {FloatData [<0 -2 0>, <0 0 0>, <0 2 0>, <-0.15 -2 -2>, <-0.15 0 -2>, <-0.15 2 -2>, <-0.3 -2 -4>, <-0.3 0 -4>, <-0.3 2 -4>]}}
     )
    )
    Skeletal Mesh=$System.Skeletal Mesh
   )
   /SetVariables(
    Particles.BeamColor:LinearColor=&MultiplyLinearColor(Linear Color=$Particles.Color, Scale Factor=&MakeVector4_FromFloat(Float=0.2))
    Particles.FlareColor=&MultiplyLinearColor(Linear Color=$Particles.Color, Scale Factor=&MakeVector4_FromFloat(Float=1))
    Particles.GlowColor=&MultiplyLinearColor(Linear Color=$Particles.Color, Scale Factor=&MakeVector4_FromFloat(Float=5))
    Particles.TeardropGlowColor:LinearColor=&MultiplyLinearColor(Linear Color=$Particles.Color, Scale Factor=&MakeVector4_FromFloat(Float=0.8))
   ) {isSetParameters=true}
   /SetVariables(Particles.GlowPosition:NiagaraPosition=&AddVectorToPosition(Position=$Particles.Position, Vector=<0 0 -1>)) {isSetParameters=true}
   &FN_Light_Attributes(
    Color=&MakeLinearColorFromVectorAndFloat(
     "Vector (RGB)"=&MakeVectorFromLinearColorRGB(Color=<rgba 1 0 0 1>)
     "Float (Alpha)"=$Emitter.BackLightStrength
    )
    Create Color Attribute=true
    Create Exponent Attribute=true
    Create Position Attribute=true
    Create Radius Attribute=true
    Exponent=2
    Position=&AddVectorToPosition(
     Position=$Output.SkeletalMeshLocation.SampledPositionWithoutOffset
     Vector=&RotateVectorByQuaternion(Vector To Rotate=<13 0 -15>, Quaternion=$Particles.SkeletalMeshLocation.SampledRotation)
    )
    Radius=&Multiply_Float(A=10, B=$Emitter.BackLightStrength)
    Scale Color=3
   )
  ]
  rendererClasses [&NiagaraSpriteRendererProperties, &NiagaraLightRendererProperties]
  renderers [
   {rendererClass=&NiagaraSpriteRendererProperties}
   {rendererClass=&NiagaraSpriteRendererProperties, rendererIndex=1}
   {rendererClass=&NiagaraSpriteRendererProperties, rendererIndex=2}
   {rendererClass=&NiagaraLightRendererProperties, rendererIndex=3}
  ]
 }
 NE_BurstSmokes {
  emitterUpdateScript [
   &EmitterState(Inactive Response=::NewEnumerator1, Life Cycle Mode=::NewEnumerator1, Loop Delay=0.15, Loop Duration=0, UseLoopDelay=false)
   &SpawnRate(
    Spawn Probability=&FloatFromCurve(
     FloatCurve=DataInterface() {propertyValues {Curve {keys [<at 0 0 RCIM_Constant>, <at 0.11 1>, <at 0.25 0>, <at 0.39 0>, <at 0.4 1>, {arriveTangent=-2.5249, leaveTangent=-2.5249, time=0.5, value=0}]}}}
     CurveIndex=$System.Timeline.Time
     Scale Curve=&FN_FrontEndFloatModifier(Gameplay=$User.ForwardDot, Front End=1)
    )
    SpawnRate=200
    Use Spawn Probability=true
   )
   &SpawnRate(
    Spawn Probability=&FloatFromCurve(
     FloatCurve=DataInterface() {propertyValues {Curve {keys [
         {interpMode=RCIM_Cubic, tangentMode=RCTM_User, time=0, value=0}
         {arriveTangent=4.4334, interpMode=RCIM_Cubic, leaveTangent=4.4334, tangentMode=RCTM_User, time=0.5, value=1}
        ]}}}
     CurveIndex=$System.Timeline.Time
     Scale Curve=&FN_FrontEndFloatModifier(Gameplay=$User.ForwardDot, Front End=1)
    )
    SpawnGroup=1
    SpawnRate=10
    Use Spawn Probability=true
   ) {module=SpawnRate001}
  ]
  particleSpawnScript [
   &InitializeParticle(
    Color=&MakeLinearColorFromVectorAndFloat("Vector (RGB)"=<1 1 1>, "Float (Alpha)"=$System.Timeline.Time)
    Lifetime=&MergeHelper_FloatSwitch(
     0=&RandomRangeFloat(Minimum=0.5, Maximum=0.6)
     1=&RandomRangeFloat(Minimum=0.6, Maximum=0.8)
     2=?
     3=?
     Index=&SpawnGroup()
     Num=::NewEnumerator0
    )
    Position Offset=$System.JetPosition
    Sprite Size Mode=::NewEnumerator3
    Uniform Sprite Size=&Multiply_Float(
     A=&MergeHelper_FloatSwitch(0=80, 1=50, 2=?, 3=?, Index=&SpawnGroup(), Num=::NewEnumerator0)
     B=&MakeFloatFromVector(Vector=$Engine.Owner.Scale, Channel=::NewEnumerator0)
    )
    UsePositionOffset=true
    Write Material Random=true
   )
   &SkeletalMeshLocation(Bone Sampling Mode=::NewEnumerator4, MaskSpawnGroup=false, Sampled Position Offset=<16 0 0>, Skeletal Mesh=$System.Skeletal Mesh)
   /SetVariables(Particles.PerParticleRandom:NiagaraFloat=&RandomRangeFloat(Minimum=0, Maximum=1)) {isSetParameters=true}
  ]
  particleUpdateScript [
   &ParticleState()
   &Color(
    Color=&MergeHelper_ColorSwitch(
     0=&ColorFromCurve(ColorCurve=DataInterface() {propertyValues {
        AlphaCurve {keys [<at 0 0.2>,<at 0.1474 0.1>,<at 0.75 0.05>,<at 1 0>]}
        BlueCurve {keys [<at 0.0061 0.7552>, <at 0.1705 0>, {arriveTangent=0.0577, time=0.2872, value=0.3542}]}
        GreenCurve {keys [<at 0.0061 0.3475>, <at 0.1705 0.1283>, {arriveTangent=-0.0961, time=0.2872, value=0.3542}]}
        RedCurve {keys [<at 0.0061 0>, <at 0.1705 0.5156>, {arriveTangent=-1.0957, time=0.2872, value=0.3542}]}
       }})
     1=&ColorFromCurve(ColorCurve=DataInterface() {propertyValues {
        AlphaCurve {keys [<at 0 0.05>,<at 0.75 0.02>,<at 1 0>]}
        BlueCurve {keys [{arriveTangent=0.0577, time=0, value=0.3542}]}
        GreenCurve {keys [{arriveTangent=-0.0961, time=0, value=0.3542}]}
        RedCurve {keys [{arriveTangent=-1.0957, time=0, value=0.3542}]}
       }})
     2=?
     3=?
     Index=&SpawnGroup()
     Num=::NewEnumerator0
    )
    Scale Alpha=&Multiply_Float(
     A=&MergeHelper_FloatSwitch(
      0=&RandomRangeFloat(Minimum=0.7, Maximum=0.8)
      1=1
      2=?
      3=?
      Index=&SpawnGroup()
      Num=::NewEnumerator0
     )
     B=&MakeFloatFromLinearColor(LinearColor=$Particles.Initial.Color, Channel:ENiagaraFloatFromLinearColorOptions=::NewEnumerator3)
    )
   )
   &ScaleSpriteSize(
    Scale Sprite Size Mode=::NewEnumerator0
    Uniform Scale Factor=&MergeHelper_FloatSwitch(
     0=&FloatFromCurve(
      FloatCurve=DataInterface() {propertyValues {Curve {keys [<at 0 0>, {arriveTangent=-5.0995, time=1, value=1}]}}}
      Scale Curve=&Lerp_Float(A=0.7, B=1, Alpha=$Particles.PerParticleRandom)
     )
     1=&FloatFromCurve(
      FloatCurve=DataInterface() {propertyValues {Curve {keys [
          {arriveTangent=2.1219, interpMode=RCIM_Cubic, leaveTangent=2.1219, tangentMode=RCTM_User, time=0, value=0.3}
          {interpMode=RCIM_Cubic, tangentMode=RCTM_User, time=1, value=1.1}
         ]}}}
      Scale Curve=&Lerp_Float(A=0.5, B=1, Alpha=$Particles.PerParticleRandom)
     )
     2=?
     3=?
     Index=&SpawnGroup()
     Num=::NewEnumerator0
    )
   )
   &DynamicMaterialParameters(
    Index 0 Param 1=&Add_Float(
     A=&Multiply_Float(A=$Engine.Time, B=1)
     B=&RandomRangeFloat(Minimum=0, Maximum=1)
    )
    Index 0 Param 3=&FloatFromCurve(
     FloatCurve=DataInterface() {propertyValues {Curve {keys [<at 0 0>,<at 0.333 0>]}}}
     CurveIndex=$Particles.Age
     Scale Curve=&Multiply_Float(A=0.02, B=$System.Timeline.Time)
    )
    Index 2 Param 1=1
    Index 2 Param 2=&RandomRangeFloat(Minimum=0, Maximum=1)
    Index 2 Param 3=&FloatFromCurve(FloatCurve=DataInterface() {propertyValues {Curve {keys [
         {interpMode=RCIM_Cubic, tangentMode=RCTM_Break, time=0.666, value=0}
         {arriveTangent=10.9325, interpMode=RCIM_Cubic, tangentMode=RCTM_Break, time=1, value=2}
        ]}}})
    Index 2 Param 4=1
    Module.Param0WriteEnabled=true
    Module.Param0WriteEnabled1=true
    Module.Param0WriteEnabled2=true
    Module.Param1WriteEnabled=true
    Module.Param1WriteEnabled1=true
    Module.Param1WriteEnabled2=true
    Module.Param2WriteEnabled=true
    Module.Param2WriteEnabled1=true
    Module.Param2WriteEnabled2=true
    Module.Param3WriteEnabled=true
    Module.Param3WriteEnabled1=true
    Module.Param3WriteEnabled2=true
    Write Parameter Index 1=true
    Write Parameter Index 2=true
   )
   /SetVariables(Particles.Position=&MergeHelper_VectorSwitch(
     0=&AddVectorToPosition(
      Position=$Particles.Position
      Vector=&Multiply_VectorByFloat(
       Vector=&NormalizeVector(
        Vector To Normalize=&Lerp_Vector(
         A=<0 1 0>
         B=&TransformVector(
          Vector=&NormalizeVector(
           Vector To Normalize=&Multiply_Vector(A=$Engine.Owner.Velocity, B=<-1 -1 -3>)
           Safe Normalize=true
           Safe Fallback Threshold=0
           Safe Fallback Vector=<0 0 0.001>
          )
          Source Space=::World
          Destination Space=::Local
         )
         Alpha=&FloatFromCurve(
          FloatCurve=DataInterface() {propertyValues {Curve {keys [{interpMode=RCIM_Cubic, leaveTangent=2.5306, tangentMode=RCTM_Break, time=0, value=0}, {tangentMode=RCTM_User, time=1, value=1}]}}}
          Scale Curve=$System.ForwardAndRightDot
         )
        )
        Safe Normalize=true
        Safe Fallback Threshold=0
        Safe Fallback Vector=<0 0 0.001>
       )
       Float=&Lerp_Float(A=1, B=6, Alpha=$Particles.PerParticleRandom)
      )
     )
     1=&AddVectorToPosition(
      Position=$Particles.Position
      Vector=&Multiply_VectorByFloat(
       Vector=&NormalizeVector(
        Vector To Normalize=&Lerp_Vector(
         A=<0 1 0>
         B=&TransformVector(
          Vector=&NormalizeVector(
           Vector To Normalize=&Multiply_Vector(A=$Engine.Owner.Velocity, B=<-1 -1 -3>)
           Safe Normalize=true
           Safe Fallback Threshold=0
           Safe Fallback Vector=<0 0 0.001>
          )
          Source Space=::World
          Destination Space=::Local
         )
         Alpha=&FloatFromCurve(
          FloatCurve=DataInterface() {propertyValues {Curve {keys [{interpMode=RCIM_Cubic, leaveTangent=2.5306, tangentMode=RCTM_Break, time=0, value=0}, {tangentMode=RCTM_User, time=1, value=1}]}}}
          Scale Curve=$System.ForwardAndRightDot
         )
        )
        Safe Normalize=true
        Safe Fallback Threshold=0
        Safe Fallback Vector=<0 0 0.001>
       )
       Float=2
      )
     )
     2=?
     3=?
     Index=&SpawnGroup()
     Num=::NewEnumerator0
    )) {isSetParameters=true}
   &Drag(Use Linear Drag=true, Use Rotational Drag=true)
   &SolveForcesAndVelocity(Clamp Velocity=false, Limit Acceleration=false)
  ]
  rendererClasses [&NiagaraSpriteRendererProperties]
  renderers [
   {rendererClass=&NiagaraSpriteRendererProperties}
   {rendererClass=&NiagaraSpriteRendererProperties, rendererIndex=1}
  ]
 }
 NE_SoftTrail {
  emitterSpawnScript [&FN_RibbonIDOffsetUtility()]
  emitterUpdateScript [
   &EmitterState()
   &FN_SpawnAnimTrail(Enable:NiagaraBool=&SetBoolByFloatComparison(A=$System.Timeline.NormalizedTime, B=0), Number of Anim Trails=2)
  ]
  particleSpawnScript [
   &FN_InitializeAnimTrailBoneIndecies(First Bone or Socket Index=1, Second Bone or Socket Index=0)
   &FN_InitializeAnimTrailBoneIndecies(AnimTrail Selection=1, First Bone or Socket Index=3, Second Bone or Socket Index=2) {module=FN_InitializeAnimTrailBoneIndecies001}
   /SetVariables(Particles.AlphaFadeIn:NiagaraFloat=&FloatFromCurve(FloatCurve=DataInterface() {propertyValues {Curve {keys [<at 0 0>,<at 0.5 1>]}}}, CurveIndex=$Emitter.Age)) {isSetParameters=true}
   &FN_InitializeAnimTrail(
    AnimTrailWidth=&FN_FrontEndFloatModifier(Gameplay=1, Front End=0.8)
    Color=&MakeLinearColor(
     R:NiagaraFloat=1
     G:NiagaraFloat=1
     B:NiagaraFloat=1
     A:NiagaraFloat=&Multiply_Float(A=$Particles.AlphaFadeIn, B=1)
    )
    Position Offset=&FN_FrontEndVector3Modifier(
     Gameplay=&MakeVector(X=&MaskFloatBySpawnGroup(Masked Float=2, Passthrough Float=-2), Y=-40, Z=0)
     Front End=<0 0 0>
    )
   )
   /SetVariables(Particles.Position=&AddVectorToPosition(
     Position=$Particles.Position
     Vector=&FN_FrontEndVector3Modifier(Gameplay=<0 0 0>, Front End=<0 -20 0>)
    )) {enabled=false, isSetParameters=true}
   &AddVelocity(Velocity=&Multiply_VectorByFloat(Vector=<0 800 0>, Float=&MakeCustomFloatFromBool(Bool=$User.bIsFrontEnd, True Float=1, False Float=1)))
  ]
  particleUpdateScript [
   &SolveForcesAndVelocity(Clamp Velocity=false, Limit Acceleration=false)
   &FN_UpdateAnimTrail(
    Locked Particle Color:LinearColor=&MultiplyLinearColor(
     Linear Color=$Particles.Initial.Color
     Scale Factor=&MakeFloat4FromVectorAndFloat("Vector (RGB)"=<1 1 1>, "Float (Alpha)"=$System.Timeline.Time)
    )
    Locked Particle RibbonWidth=1
    SkeletalMesh=$User.SkeletalMesh
   ) {module=FN_UpdateAnimTrail001}
   &Color(
    Color=&MultiplyLinearColor(
     Linear Color=&SelectLinearColorFromArray(
      Array Sampling Mode=::NewEnumerator1
      Direct Array Index=&MakeIntFromFloat(Float=$User.VariantPicker, Conversion Mode=::NewEnumerator1)
      Color Selection Array=DataInterface() {propertyValues {ColorData [
         <rgba 0.1 0.1 0.1 2>
         <rgba 0.1355 0.0667 0.2083 2>
         <rgba 0 0.349 0.313 2>
         <rgba 0.0183 0.2782 0.4 2>
         <rgba 0 0.599 0.0207 2>
         <rgba 0.0602 0.1613 0.4 2>
         <rgba 0.05 0.05 0.05 2>
         <rgba 0.0382 0.242 0.5 2>
         <rgba 0.1406 0.0057 0 2>
         <rgba 0.3854 0.019 0 2>
         <rgba 0.5 0.1377 0 2>
         <rgba 0.0556 0.0623 0.2135 2>
        ]}}
      Random Seed=?
     )
     Scale Factor=&MakeVector4_FromLinearColor(LinearColor=&FNTimeOfDayColor(
       Night Color=&SelectLinearColorFromArray(
        Array Sampling Mode=::NewEnumerator1
        Direct Array Index=&MakeIntFromFloat(Float=$User.VariantPicker, Conversion Mode=::NewEnumerator1)
        Color Selection Array=DataInterface() {propertyValues {ColorData [
           <rgba 1 1 1 1>
           <rgba 1 1 1 1>
           <rgba 0.3906 0.543 1 1>
           <rgba 1 1 1 1>
           <rgba 0.4118 0.6538 1 1>
           <rgba 1 1 1 1>
           <rgba 0.3906 0.543 1 1>
           <rgba 0.3906 0.543 1 1>
           <rgba 1 1 1 1>
           <rgba 0.3906 0.543 1 1>
           <rgba 0.3906 0.543 1 1>
           <rgba 0.3906 0.543 1 1>
          ]}}
        Random Seed=?
       )
       Morning Color=<rgba 1 1 1 1>
       Day Color=<rgba 1 1 1 1>
       Evening Color=<rgba 1 1 1 1>
      ))
    )
    Scale Alpha=&Multiply_Float(
     A=&Multiply_Float(A=0.5, B=$Particles.AlphaFadeIn)
     B=&FN_FrontEndFloatModifier(Gameplay=$System.Timeline.Time, Front End=1)
    )
   ) {module=Color001}
   &KillParticles(Kill Particles=&SetBoolByFloatComparison(A=$System.Timeline.NormalizedTime, B=0, Comparison Type=::NewEnumerator4))
  ]
  rendererClasses [&NiagaraRibbonRendererProperties]
  renderers [
   {rendererClass=&NiagaraRibbonRendererProperties}
   {rendererClass=&NiagaraRibbonRendererProperties, rendererIndex=1}
  ]
 }
 NE_TipTrail {
  emitterSpawnScript [&FN_RibbonIDOffsetUtility()]
  emitterUpdateScript [
   &EmitterState()
   &FN_SpawnAnimTrail(Enable:NiagaraBool=&SetBoolByFloatComparison(A=$System.Timeline.NormalizedTime, B=0), Number of Anim Trails=2)
  ]
  particleSpawnScript [
   &FN_InitializeAnimTrailBoneIndecies(First Bone or Socket Index=1, Second Bone or Socket Index=4)
   &FN_InitializeAnimTrailBoneIndecies(AnimTrail Selection=1, First Bone or Socket Index=3, Second Bone or Socket Index=5) {module=FN_InitializeAnimTrailBoneIndecies001}
   /SetVariables(Particles.AlphaFadeIn:NiagaraFloat=&FloatFromCurve(FloatCurve=DataInterface() {propertyValues {Curve {keys [<at 0 0>,<at 0.5 1>]}}}, CurveIndex=$Emitter.Age)) {isSetParameters=true}
   &FN_InitializeAnimTrail(
    AnimTrailWidth=&FN_FrontEndFloatModifier(Gameplay=1, Front End=0.75)
    Color=&MakeLinearColor(
     R:NiagaraFloat=1
     G:NiagaraFloat=1
     B:NiagaraFloat=1
     A:NiagaraFloat=&Multiply_Float(A=$Particles.AlphaFadeIn, B=0.5)
    )
    Position Offset=&FN_FrontEndVector3Modifier(Gameplay=<0 0 0>, Front End=<0 0 0>)
   )
   &AddVelocity(Velocity=&Multiply_VectorByFloat(Vector=<0 800 0>, Float=&MakeCustomFloatFromBool(Bool=$User.bIsFrontEnd, True Float=1, False Float=1)))
   &DynamicMaterialParameters(
    Index 0 Param 1=1
    Module.Param0WriteEnabled=true
    Module.Param1WriteEnabled=true
    Module.Param2WriteEnabled=true
    Module.Param3WriteEnabled=true
   )
   /SetVariables() {enabled=false, isSetParameters=true}
   &CameraOffset()
  ]
  particleUpdateScript [
   &SolveForcesAndVelocity(Clamp Velocity=false, Limit Acceleration=false)
   &FN_UpdateAnimTrail(
    Locked Particle Color:LinearColor=&MultiplyLinearColor(
     Linear Color=$Particles.Initial.Color
     Scale Factor=&MakeFloat4FromVectorAndFloat("Vector (RGB)"=<1 1 1>, "Float (Alpha)"=$System.Timeline.Time)
    )
    Locked Particle RibbonWidth=1
    SkeletalMesh=$User.SkeletalMesh
   ) {module=FN_UpdateAnimTrail001}
   /SetVariables(Particles.RibbonU0Override=$Particles.NormalizedAge) {isSetParameters=true}
   &Color(
    Color=&MultiplyLinearColor(
     Linear Color=&ColorFromCurve(ColorCurve=DataInterface() {propertyValues {
        AlphaCurve {keys [<at 0 1 RCIM_Cubic>]}
        BlueCurve {keys [<at 0.003 0.1>,<at 1 0.5 RCIM_Cubic>]}
        GreenCurve {keys [<at 0.003 0.1>,<at 1 0.5 RCIM_Cubic>]}
        RedCurve {keys [<at 0.003 0.1>,<at 1 0.5 RCIM_Cubic>]}
       }})
     Scale Factor=&MakeVector4_FromLinearColor(LinearColor=&SelectLinearColorFromArray(
       Array Sampling Mode=::NewEnumerator1
       Direct Array Index=&MakeIntFromFloat(Float=$User.VariantPicker, Conversion Mode=::NewEnumerator1)
       Color Selection Array=DataInterface() {propertyValues {ColorData [
          <rgba 0.1 0.1 0.1 1>
          <rgba 0.1 0.1 0.1 1>
          <rgba 0.1 0.1 0.1 1>
          <rgba 0.1 0.1 0.1 1>
          <rgba 0.1 0.1 0.1 1>
          <rgba 0.1 0.1 0.1 1>
          <rgba 0.5 0 0 1>
          <rgba 0.7 0.1865 0.5858 1>
          <rgba 0.1 0.1 0.1 1>
          <rgba 0.1 0.1 0.1 1>
          <rgba 2 0.5 0 1>
          <rgba 0 0.2378 1.5 1>
         ]}}
       Random Seed=?
      ))
    )
    Scale Alpha=&Multiply_Float(
     A=&Multiply_Float(A=0.5, B=$Particles.AlphaFadeIn)
     B=&FN_FrontEndFloatModifier(Gameplay=$System.Timeline.Time, Front End=1)
    )
    Scale Color=&VectorFromFloat(Value=&FNTimeOfDayScalar(Night Scale=0.25, Morning Scale=0.5, Evening Scale=0.5))
   ) {module=Color001}
   &KillParticles(Kill Particles=&SetBoolByFloatComparison(A=$System.Timeline.NormalizedTime, B=0, Comparison Type=::NewEnumerator4))
  ]
  rendererClasses [&NiagaraRibbonRendererProperties]
  renderers [{rendererClass=&NiagaraRibbonRendererProperties}]
 }
 __NE_FlameBurst {
  emitterUpdateScript [
   &EmitterState(
    Inactive Response=::NewEnumerator1
    Life Cycle Mode=::NewEnumerator1
    Loop Behavior=::NewEnumerator1
    Loop Duration Mode=::NewEnumerator1
    UseLoopDelay=false
   )
   &SpawnBurst_Instantaneous(Use Spawn Probability=false, UseLoopCountLimit=false)
   /SetVariables(Emitter.DualBurstFloat:NiagaraFloat=&FloatFromCurve(
     FloatCurve=DataInterface() {propertyValues {Curve {keys [
         <at 0 0 RCIM_Constant>
         <at 0.11 0>
         <at 0.25 1>
         <at 0.26 0>
         <at 0.39 0>
         <at 0.49 1>
         {arriveTangent=-2.5249, leaveTangent=-2.5249, time=0.5, value=0}
        ]}}}
     CurveIndex=$System.Timeline.Time
     Scale Curve=&FN_FrontEndFloatModifier(Gameplay=$User.ForwardDot, Front End=1)
    )) {isSetParameters=true}
  ]
  particleSpawnScript [
   &InitializeParticle(
    Mesh Scale=<2 4 2>
    Mesh Scale Mode=::NewEnumerator4
    Sprite Rotation Angle=-0.25
    Sprite Rotation Mode=::NewEnumerator1
    Sprite Size Max=<25 20>
    Sprite Size Min=<25 20>
    Sprite Size Mode=::NewEnumerator2
    Sprite UV Mode=::NewEnumerator2
    UsePositionOffset=true
    Write Material Random=true
   )
   &AddVelocity(Velocity=<0 1 0>)
   &CameraOffset(Camera Offset Amount=&FN_FrontEndFloatModifier(Gameplay=5, Front End=10))
   &SpriteFacingAndAlignment(Sprite Facing=<0 0 0>, WriteAlignment=true, WriteFacing=false)
  ]
  particleUpdateScript [
   &ParticleState(Kill Particles When Lifetime Has Elapsed=false, Let Infinitely Lived Particles Die When Emitter Deactivates=true)
   &ScaleSpriteSize(
    Non-Uniform Curve Index:NiagaraFloat=$Emitter.DualBurstFloat
    Non-Uniform Curve Sprite Scale=DataInterface() {propertyValues {
      XCurve {keys [{tangentMode=RCTM_User, time=0, value=1.5}, {tangentMode=RCTM_User, time=1, value=2}]}
      YCurve {keys [{tangentMode=RCTM_User, time=0, value=0.5}, {tangentMode=RCTM_User, time=1, value=1}]}
     }}
    Scale Sprite Size Mode=::NewEnumerator3
   )
   &Drag(Drag=4, Use Linear Drag=true, Use Rotational Drag=true)
   &SolveForcesAndVelocity(Clamp Velocity=false, Limit Acceleration=false)
   &Color(Color=&MakeLinearColorFromVectorAndFloat(
     "Vector (RGB)"=&MakeVectorFromLinearColorRGB(Color=&ColorFromCurve(
       ColorCurve=DataInterface() {propertyValues {
         AlphaCurve {keys [<at 0 1 RCIM_Cubic>]}
         BlueCurve {keys [<at 0.5567 5>,<at 0.5958 0>,<at 0.9823 0>]}
         GreenCurve {keys [<at 0.5567 0.2>,<at 0.5958 1.4159>,<at 0.9823 0.0714>]}
         RedCurve {keys [<at 0.5567 0>,<at 0.5958 5>,<at 0.9823 0.6406>]}
        }}
       CurveIndex=$Emitter.DualBurstFloat
      ))
     "Float (Alpha)"=&Multiply_Float(
      A=&FloatFromCurve(FloatCurve=DataInterface() {propertyValues {Curve {keys [<at 0 0>,<at 0.1 1>]}}}, CurveIndex=$Emitter.DualBurstFloat)
      B=&FloatFromCurve(
       FloatCurve=DataInterface() {propertyValues {Curve {keys [<at 0.95 1>,<at 1 0>]}}}
       CurveIndex=&Abs_Float(Float:NiagaraFloat=&DotProduct(Vector A=&ReturnCameraVector(Camera Vectors=::NewEnumerator1, Coordinate Space=::Simulation), Vector B=<0 1 0>))
      )
     )
    ))
   &SkeletalMeshLocation(Bone Sampling Mode=::NewEnumerator4, MaskSpawnGroup=false, Sampled Position Offset=<16 0 0>, Skeletal Mesh=$System.Skeletal Mesh)
   /SetVariables(
    Particles.RadialFlameColor:LinearColor=&MakeLinearColorFromVectorAndFloat(
     "Vector (RGB)"=&MakeVectorFromLinearColorRGB(Color=$Particles.Color)
     "Float (Alpha)"=&Multiply_Float(
      A=&FloatFromCurve(FloatCurve=DataInterface() {propertyValues {Curve {keys [<at 0 0>,<at 0.1 1>]}}}, CurveIndex=$Emitter.DualBurstFloat)
      B=&FloatFromCurve(
       FloatCurve=DataInterface() {propertyValues {Curve {keys [<at 0.5 0>,<at 0.55 1>]}}}
       CurveIndex=&Abs_Float(Float:NiagaraFloat=&DotProduct(Vector A=&ReturnCameraVector(Camera Vectors=::NewEnumerator1, Coordinate Space=::Simulation), Vector B=<0 1 0>))
      )
     )
    )
    Particles.RadialFlameFacing:Vector3f=<0 1 0>
    Particles.RadialFlameSize:Vector2f=&Multiply_Vector2D_ByFloat(
     Float=&FloatFromCurve(FloatCurve=DataInterface() {propertyValues {Curve {keys [<at 0 1>,<at 1 1>]}}}, CurveIndex=$Emitter.DualBurstFloat)
     Vector2D=<20 20>
    )
    Particles.RadialFlareAlignment:Vector3f=<1 0 0>
    Particles.RadialFlareRotation:NiagaraFloat=&RandomRangeFloat(Minimum=0, Maximum=360)
   ) {isSetParameters=true}
   &DynamicMaterialParameters(
    Index 0 Param 1=&Add_Float(
     A=&RandomRangeFloat(Minimum=0, Maximum=1)
     B=&FloatFromCurve(FloatCurve=DataInterface() {propertyValues {Curve {keys [<at 0 0>,<at 1 1>]}}}, Scale Curve=0.1)
    )
    Index 1 Param 3=&FloatFromCurve(FloatCurve=DataInterface() {propertyValues {Curve {keys [<at 0 0>,<at 1.2 1>]}}})
    Module.Index0Param0Write=true
    Module.Index0Param1Write=true
    Module.Index0Param2Write=true
    Module.Index0Param3Write=true
    Module.Index1Param0Write=true
    Module.Index1Param1Write=true
    Module.Index1Param2Write=true
    Module.Index1Param3Write=true
    Write Parameter Index 1=true
   )
   &FN_Sprite_Attributes(
    Alignment:Vector3f=<1 0 0>
    Color=&MakeLinearColorFromVectorAndFloat(
     "Vector (RGB)"=&MakeVectorFromLinearColorRGB(Color=&ColorFromCurve(
       ColorCurve=DataInterface() {propertyValues {
         AlphaCurve {keys [<at 0 1>]}
         BlueCurve {keys [<at 0.3867 15>,<at 0.6206 0>]}
         GreenCurve {keys [<at 0.3867 2.3399>,<at 0.6206 1.1901>]}
         RedCurve {keys [<at 0.3867 0>,<at 0.6206 5>]}
        }}
       CurveIndex=$System.Timeline.Time
      ))
     "Float (Alpha)"=&FloatFromCurve(
      FloatCurve=DataInterface() {propertyValues {Curve {keys [
          {arriveTangent=25.1156, interpMode=RCIM_Cubic, leaveTangent=25.1156, tangentMode=RCTM_User, time=0, value=0}
          <at 0.1 1 RCIM_Cubic>
          {arriveTangent=-0.4286, interpMode=RCIM_Cubic, leaveTangent=-0.4286, tangentMode=RCTM_User, time=1, value=0}
         ]}}}
      CurveIndex=$System.Timeline.Time
      Scale Curve=&FN_FrontEndFloatModifier(Gameplay=$User.ForwardDot, Front End=1)
     )
    )
    Create Alignment Attribute=true
    Create Color Attribute=true
    Create Facing Attribute=true
    Create SpriteSize Attribute=true
    Facing:Vector3f=<0 1 0>
    SpriteSize=&Vector2DFromFloat(Value=&Multiply_Float(A=10, B=&MakeFloatFromVector(Vector=$Engine.Owner.Scale, Channel=::NewEnumerator0)))
    Write to Module Namespace=true
   )
   &FN_Light_Attributes(
    Color=&MakeLinearColorFromVectorAndFloat(
     "Vector (RGB)"=&Multiply_VectorByFloat(
      Vector=&MakeVectorFromLinearColorRGB(Color=&ColorFromCurve(
        ColorCurve=DataInterface() {propertyValues {
          AlphaCurve {keys [<at 0 1>]}
          BlueCurve {keys [<at 0 0>,<at 0.172 0>,<at 0.3165 0>,<at 0.4855 0>]}
          GreenCurve {keys [<at 0 1.4682>,<at 0.172 1>,<at 0.3165 1.7667>,<at 0.4855 1>]}
          RedCurve {keys [<at 0 15>,<at 0.172 5>,<at 0.3165 15>,<at 0.4855 5>]}
         }}
        CurveIndex=$System.Timeline.Time
       ))
      Float=4
     )
     "Float (Alpha)"=1
    )
    Create Color Attribute=true
    Create Exponent Attribute=true
    Create Position Attribute=true
    Create Radius Attribute=true
    Enable Particle Light Enable=true
    Exponent=1.5
    Light Enabled:NiagaraBool=&SetBoolByFloatComparison(A=$System.Timeline.Time, B=0)
    Position=&AddVectorToPosition(
     Position=$Output.SkeletalMeshLocation.SampledPositionWithoutOffset
     Vector=&RotateVectorByQuaternion(Vector To Rotate=<20 0 0>, Quaternion=$Particles.SkeletalMeshLocation.SampledRotation)
    )
    Radius=&Multiply_Float(A=&MakeFloatFromVector(Vector=$Engine.Owner.Scale, Channel=::NewEnumerator0), B=125)
    Scale Color=&FloatFromCurve(
     FloatCurve=DataInterface() {propertyValues {Curve {keys [
         {arriveTangent=25.1156, interpMode=RCIM_Cubic, leaveTangent=25.1156, tangentMode=RCTM_User, time=0, value=0}
         <at 0.1 1 RCIM_Cubic>
         {arriveTangent=-0.4286, interpMode=RCIM_Cubic, leaveTangent=-0.4286, tangentMode=RCTM_User, time=1, value=0}
        ]}}}
     CurveIndex=$System.Timeline.Time
     Scale Curve=&FN_FrontEndFloatModifier(Gameplay=$User.ForwardDot, Front End=1)
    )
   )
  ]
  rendererClasses [&NiagaraSpriteRendererProperties, &NiagaraLightRendererProperties]
  renderers [
   {rendererClass=&NiagaraSpriteRendererProperties}
   {rendererClass=&NiagaraSpriteRendererProperties, rendererIndex=1}
   {rendererClass=&NiagaraSpriteRendererProperties, rendererIndex=2}
   {rendererClass=&NiagaraLightRendererProperties, rendererIndex=3}
  ]
 }
}
summary {
 systemName=NS_Glider_GelatoCourage
 emitters [
  {emitterName=EmbersGPUHigh, simTarget=GPUComputeSim, rendererClasses [&NiagaraSpriteRendererProperties]}
  {emitterName=NE_BurstSmokes, rendererClasses [&NiagaraSpriteRendererProperties]}
  {emitterName=NE_Backlights, rendererClasses [&NiagaraSpriteRendererProperties, &NiagaraLightRendererProperties]}
  {emitterName=NE_SoftTrail, rendererClasses [&NiagaraRibbonRendererProperties]}
  {emitterName=NE_TipTrail, rendererClasses [&NiagaraRibbonRendererProperties]}
  {
   emitterName=__NE_FlameBurst
   rendererClasses [&NiagaraSpriteRendererProperties, &NiagaraLightRendererProperties]
  }
 ]
 userVariables [
  User.bIsFrontEnd {type=NiagaraBool}
  User.bIsFrontEndPreview {type=NiagaraBool}
  User.bIsFullyDeployed {type=NiagaraBool}
  User.ForwardDot {type=NiagaraFloat, defaultValue=0, description=""}
  User.RibbonIDOffset {type=NiagaraInt32}
  User.RightDot {type=NiagaraFloat, defaultValue=0, description=""}
  User.SkeletalMesh {
   type=NiagaraDataInterfaceSkeletalMesh
   defaultValue {
    struct=NiagaraExt_VariableValue_DataInterface
    value {dataInterface=&self:NiagaraDataInterfaceSkeletalMesh_1}
   }
  }
  User.VariantPicker {type=NiagaraFloat, defaultValue=0, description=""}
 ]
}
